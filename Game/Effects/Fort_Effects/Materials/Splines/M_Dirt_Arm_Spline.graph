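# Translucent unlit material for Niagara mesh particles, rendering a panning dirt or smoke arm, thickened and tapered by SplineThicken.
# Panning ribbon smoke and cloud detail textures build opacity and normals, with erosion, depth and near-camera fades.
# Fake directional light, fresnel and edge glow drive an emissive ramp tinted by particle color and optional time-of-day colors.
# Dynamic parameters control pan speed, glow, width and erosion; mobile and low-quality paths drop cloud detail.
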Material /Game/Effects/Fort_Effects/Materials/Splines/M_Dirt_Arm_Spline
graph {
 properties {
  bUsedWithNiagaraMeshParticles=true
  blendMode=BLEND_Translucent
  pixelDepthOffsetMode=PDOM_Legacy
  shadingModel=MSM_Unlit
  twoSided=true
  usedShadingModels=Unlit
 }
 nodes [
  n1 = ScalarParameter() {defaultValue=0.3, parameterName=Value Min}
  n2 = ScalarParameter() {defaultValue=3, parameterName=Value Max}
  n3 = ScalarParameter() {defaultValue=0.5, parameterName=DetailTileX}
  n4 = ScalarParameter() {defaultValue=2, parameterName=DetailTileY}
  n5 = Multiply(
   A=Multiply(A=Time(), B=ScalarParameter() {defaultValue=1, parameterName=PanSpeed})
   B=DynamicParameter() {defaultValue=<rgba 1 1 1 0>, paramNames [PanSpeed, GlowIntensity, Width, Erode]}.PanSpeed
  )
  n6 = Panner(
   Coordinate=Multiply(A=AppendVector(A=@n3, B=@n4), B=TextureCoordinate())
   Time=Multiply(A=ScalarParameter() {defaultValue=-0.5, parameterName=DetailPanMult}, B=@n5)
  ) {speedX=0.04, speedY=0.5}
  n7 = OneMinus(None=TextureSample(UVs=@n6) {samplerType=SAMPLERTYPE_LinearGrayscale, texture=/Game/Global/Textures/Default/Tileable_Patterns/T_StormCloudDetail01_M}.G)
  n8 = Constant() {r=0.2}
  n9 = QualitySwitch(Default=Multiply(A=Multiply(A=@n7, B=@n7), B=@n8), Low=@n8)
  n10 = SkyAtmosphereLightDirection()
  n11 = Constant3Vector() {constant=<rgba 0 0 1 0>}
  n12 = Multiply(
   A=ScalarParameter() {defaultValue=100, parameterName=BaseWidth}
   B=DynamicParameter() {defaultValue=<rgba 1 1 1 0>, paramNames [PanSpeed, GlowIntensity, Width, Erode]}.Width
  )
  n13 = NamedRerouteUsage()
  n14 = MaterialFunctionCall(WidthBase=@n12, WidthTip=Multiply(A=@n12, B=ScalarParameter() {defaultValue=0.05, parameterName=WidthTipMult}), AdditionalNormal=@n13) {materialFunction=&SplineThicken}
  n15 = ConstantBiasScale(None=DotProduct(
    A=StaticSwitchParameter(True=@n10, False=@n11) {defaultValue=true, parameterName=Fake Light from Light Source}
    B=MaterialFunctionCall(Vector=@n14.Normal) {materialFunction=&SafeNormalize}.Result
   ))
  n16 = ScalarParameter() {defaultValue=7, group=Glow, parameterName=Fake Light Exp}
  n17 = Constant3Vector() {constant=<rgba 0 0 1 0>}
  n18 = OneMinus(None=Abs(None=DotProduct(A=@n13, B=@n17)))
  n19 = ScalarParameter() {defaultValue=3, group=Glow, parameterName=Fresnel Glow Exp}
  n20 = Add(
   A=Multiply(A=Power(Base=@n15, Exp=@n16), B=ScalarParameter() {defaultValue=0.3, parameterName=Fake Light Intensity})
   B=Multiply(A=Power(Base=@n18, Exp=@n19) {constExponent=3}, B=ScalarParameter() {defaultValue=0.2, group=Glow, parameterName=Fresnel Glow Strength})
  )
  n21 = TextureCoordinate()
  n22 = AppendVector(A=Constant() {r=1}, B=ScalarParameter() {defaultValue=1, parameterName=Main Texture Tiling})
  n23 = Panner(
   Coordinate=MaterialFunctionCall(XY=Multiply(A=@n21, B=@n22)) {materialFunction=&Swizzle}.YX
   Time=@n5
  ) {speedX=0.5}
  n24 = TextureSample(UVs=@n23) {texture=/Game/Effects/Fort_Effects/Gameplay/SmokeDustExplosions/Textures/T_RibbonSmokeMasked_04}
  n25 = Multiply(A=OneMinus(None=@n24.G), B=ScalarParameter() {defaultValue=0.2, parameterName=Edge Glow})
  n26 = Add(
   A=Multiply(
    A=ScalarParameter() {defaultValue=1, parameterName=Cloud Texture}
    B=ShadingPathSwitch(Default=@n9, Mobile=@n8) {desc=Migrated from FeatureLevelSwitch}
   )
   B=Multiply(A=Add(A=@n20, B=@n25), B=DynamicParameter() {defaultValue=<rgba 1 1 1 0>, paramNames [PanSpeed, GlowIntensity, Width, Erode]}.GlowIntensity)
  )
  n27 = VectorParameter() {group=ToD, parameterName=TOD Morning, defaultValue=<rgba 1 0.3343 0.122 0>}
  n28 = VectorParameter() {group=ToD, parameterName=TOD Day, defaultValue=<rgba 1 1 1 0>}
  n29 = VectorParameter() {group=ToD, parameterName=TOD Evening, defaultValue=<rgba 1 0.1673 0.1291 0>}
  n30 = VectorParameter() {group=ToD, parameterName=TOD Night, defaultValue=<rgba 0.229 0.442 1 0>}
  n31 = Multiply(
   A=Multiply(A=ParticleColor().RGB, B=LinearInterpolate(A=@n1, B=@n2, Alpha=@n26))
   B=StaticSwitchParameter(
    True=MaterialFunctionCall(Color_Morning=@n27.RGB, Color_Day=@n28.RGB, Color_Evening=@n29.RGB, Color_Night=@n30.RGB) {materialFunction=&Color_Time_of_Day}.Result
    False=Constant() {r=1}
   ) {parameterName=Use Time of Day colors}
  )
  MaterialFunctionCall(Emissive=@n31) {materialFunction=&MF_FX_EmissiveAdjustments}
  n32 = TextureCoordinate()
  n33 = ScalarParameter() {defaultValue=1, parameterName=Cloud Tiling}
  n34 = Time()
  n35 = ScalarParameter() {defaultValue=1, parameterName=Cloud Panning Speed}
  n36 = TextureSample(UVs=Panner(Coordinate=Multiply(A=@n32, B=@n33), Time=Multiply(A=@n34, B=@n35)) {speedY=1}) {samplerType=SAMPLERTYPE_LinearColor, texture=/Game/Items/Traps/Textures/T_Barrier_Clouds_M}
  Multiply(A=@n36.R, B=@n36.R)
  n37 = ScalarParameter() {defaultValue=1, parameterName=OpacityMorning}
  n38 = ScalarParameter() {defaultValue=1, parameterName=OpacityDay}
  n39 = ScalarParameter() {defaultValue=1, parameterName=OpacityEvening}
  n40 = ScalarParameter() {defaultValue=1, parameterName=OpacityNight}
  n41 = OneMinus(None=Saturate(None=DynamicParameter() {defaultValue=<rgba 1 1 1 0>, paramNames [PanSpeed, GlowIntensity, Width, Erode]}.Erode))
  n42 = Multiply(A=@n41) {constB=-1}
  n43 = Add(A=@n41, B=@n24.R)
  n44 = Multiply(A=@n24.G, B=Saturate(None=LinearInterpolate(A=@n42, B=@n43, Alpha=@n41)))
  n45 = ScalarParameter() {defaultValue=0.5, parameterName=Erode Softness, sliderMax=1, sliderMin=0.1}
  n46 = MaterialFunctionCall(Distance=ScalarParameter() {defaultValue=100, parameterName=NearCamFadeDistance}) {materialFunction=&Near_Camera_Fade}
  n47 = ShadingPathSwitch(Default=DepthFade(FadeDistance=ScalarParameter() {defaultValue=20, parameterName=DepthFadeDist}), Mobile=Constant() {r=1}) {desc=Migrated from FeatureLevelSwitch}
  n48 = ComponentMask(None=TextureCoordinate()) {g=true}
  n49 = Multiply(A=Multiply(A=Divide(A=@n44, B=@n45), B=Multiply(A=@n46.Result, B=@n47)), B=Saturate(None=LinearInterpolate(Alpha=@n48) {constA=4, constB=0}))
  n50 = StaticSwitchParameter(
   True=Multiply(
    A=MaterialFunctionCall(Color_Morning=@n37, Color_Day=@n38, Color_Evening=@n39, Color_Night=@n40) {materialFunction=&Color_Time_of_Day}.Result
    B=@n49
   )
   False=@n49
  ) {parameterName=Use Color_time_of_Day for Opacity}
  NamedRerouteDeclaration(None=Multiply(A=@n50, B=ParticleColor().A)) {name=Opacity, nodeColor=<rgba 0.2852 0.7772 1 1>}
  NamedRerouteDeclaration(None=@n14.WorldPositionOffset) {name=WPO, nodeColor=<rgba 1 0 0.039 1>}
  NamedRerouteDeclaration(None=@n31) {name=Emissive, nodeColor=<rgba 1 0.8144 0.1451 1>}
  n51 = Multiply(
   A=Multiply(
    A=VectorParameter() {parameterName=DetailNormalMult, defaultValue=<rgba 1 1 0 0>}.RGB
    B=TextureSample(UVs=@n6) {samplerType=SAMPLERTYPE_Normal, texture=/Game/Global/Textures/Default/Tileable_Patterns/T_StormCloudDetail01_N}.RGB
   )
   B=OneMinus(None=@n24.R)
  )
  n52 = Constant()
  n53 = TextureSample(UVs=@n23) {samplerType=SAMPLERTYPE_Normal, texture=/Game/Effects/Fort_Effects/Gameplay/SmokeDustExplosions/Textures/T_RibbonSmokeMasked_01}
  n54 = Add(
   A=ShadingPathSwitch(Default=QualitySwitch(Default=@n51, Low=@n52), Mobile=@n52) {desc=Migrated from FeatureLevelSwitch}
   B=MaterialFunctionCall(R=@n53.G, G=Multiply(A=@n53.R) {constB=-1}, B=@n53.G) {materialFunction=&AppendMany}.RGB
  )
  NamedRerouteDeclaration(None=@n54) {name=Normals TS, nodeColor=<rgba 0.1674 0.1911 1 1>}
  NamedRerouteUsage()
  NamedRerouteUsage()
  NamedRerouteUsage()
  Reroute(None=@n50)
  Reroute(None=@n54)
  Reroute()
 ]
}
