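# Translucent, two-sided unlit particle material for dust spike effects that fakes sun and moon lighting.
# A SplineThicken normal dotted with the sun/moon vector drives a highlight over time-of-day ambient, tinted by BaseColor and particle color.
# A randomly offset muzzle-flash texture, blended with panning noise via dynamic parameters, shapes color and alpha.
# Alpha adds depth and near-camera fades, skipped on low quality.
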Material /Game/Effects/Fort_Effects/Materials/Splines/M_Dust_Spike_Spline_FakeLit
graph {
 properties {
  bTangentSpaceNormal=false
  bUsedWithMeshParticles=true
  bUsedWithNanite=true
  bUsedWithNiagaraMeshParticles=true
  bUsedWithNiagaraSprites=true
  bUsedWithParticleSprites=true
  blendMode=BLEND_Translucent
  opacityMaskClipValue=0.933
  pixelDepthOffsetMode=PDOM_Legacy
  shadingModel=MSM_Unlit
  translucencyPass=MTP_BeforeDOF
  twoSided=true
  usedShadingModels=Unlit
 }
 nodes [
  n1 = MaterialFunctionCall() {materialFunction=&Color_Time_of_Day}
  n2 = LinearInterpolate(B=@n1.Result) {constA=1, constAlpha=0.25}
  n3 = LinearInterpolate(B=@n1.Result) {constA=0.9, constAlpha=0.6}
  n4 = CollectionParameter() {
   collection=&FortniteMaterialParameters
   parameterName=SunAndMoonModelDirectionalVector
  }
  n5 = ScalarParameter() {defaultValue=1, parameterName=WidthBaseMult}
  n6 = Multiply(
   A=DynamicParameter() {defaultValue=<rgba 0 1 1 1>, paramNames [TextureBlend,PanSpeed,Random,Width]}.Width
   B=ScalarParameter() {defaultValue=40, parameterName=Width}
  )
  n7 = DotProduct(
   A=ComponentMask(None=@n4) {b=true, g=true, r=true}
   B=MaterialFunctionCall(WidthBase=Multiply(A=@n5, B=@n6) {constB=0.5}, WidthTip=@n6) {materialFunction=&SplineThicken}.Normal
  )
  n8 = Clamp(None=Power(Base=Clamp(None=@n7) {clampMode=CMODE_ClampMin, minDefault=0.01}, Exp=ScalarParameter() {defaultValue=3, parameterName=Spec Highlight Power}) {constExponent=20})
  n9 = Multiply(A=Round(None=DynamicParameter() {defaultValue=<rgba 0 1 1 1>, paramNames [TextureBlend,PanSpeed,Random,Width]}.Random)) {constB=0.5}
  n10 = Constant() {r=0.015}
  n11 = TextureSampleParameter2D(UVs=Add(A=TextureCoordinate() {uTiling=0.5, vTiling=0.95}, B=AppendVector(A=@n9, B=@n10))) {parameterName=FlashTexture, texture=/Game/Weapons/FORT_Pistols/Effects/Muzzle/T_Muzzle_Flash_01_Sharp}
  n12 = ScalarParameter() {defaultValue=1, parameterName=UTiling}
  n13 = ScalarParameter() {defaultValue=1, parameterName=VTiling}
  n14 = DynamicParameter() {defaultValue=<rgba 0 1 1 1>, paramNames [TextureBlend,PanSpeed,Random,Width]}
  n15 = Time()
  n16 = DynamicParameter() {parameterIndex=1, defaultValue=<rgba 0 1 1 1>, paramNames [Noise Pan X, Noise Pan Y, Param3, Param4]}
  n17 = Panner(
   Coordinate=Multiply(A=TextureCoordinate() {uTiling=2, vTiling=0.5}, B=AppendVector(A=@n12, B=@n13))
   Time=Multiply(A=@n14.PanSpeed, B=Add(A=@n14.Random, B=@n15))
   Speed=AppendVector(A=@n16.Noise Pan X, B=@n16.Noise Pan X)
  ) {speedY=1}
  n18 = LinearInterpolate(
   A=@n11.R
   B=Multiply(
    A=@n11.R
    B=TextureSampleParameter2D(UVs=@n17) {parameterName=NoiseTexture, texture=/Game/Global/Textures/Default/Tileable_Patterns/T_Tornado_Tile}.R
   )
   Alpha=DynamicParameter() {defaultValue=<rgba 0 1 1 1>, paramNames [TextureBlend,PanSpeed,Random,Width]}.TextureBlend
  )
  n19 = Multiply(
   A=Add(A=Multiply(A=@n2) {constB=0.35}, B=Multiply(A=@n3, B=@n8))
   B=Multiply(A=LinearInterpolate(Alpha=@n18) {constA=0.7}, B=VectorParameter() {parameterName=BaseColor, defaultValue=<rgba 0.5679 0.5995 0.64 0>}.RGB)
  )
  n20 = ParticleColor()
  MaterialFunctionCall(Emissive=MaterialFunctionCall(Emissive=Multiply(A=@n19, B=@n20.RGB)) {materialFunction=&MF_FX_EmissiveAdjustments}.Result) {materialFunction=&MF_FX_EmissiveAdjustments}
  n21 = DepthFade(
   Opacity=MaterialFunctionCall(Distance=ScalarParameter() {defaultValue=200, parameterName=NearFadeDist}) {materialFunction=&Near_Camera_Fade}.Result
   FadeDistance=ScalarParameter() {defaultValue=40, parameterName=DepthFadeDist}
  )
  Multiply(A=Multiply(A=@n20.A, B=@n18), B=QualitySwitch(Default=@n21, Low=Constant() {r=1}, High=@n21, Medium=@n21))
 ]
}
