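# Translucent, unlit, two-sided material for smoky, glowing explosion arms on spline mesh particles, including Niagara.
# Panned ribbon-smoke textures and normals, with cloud detail dropped on low quality and mobile, drive fake lighting around a sphere-mask glow.
# Glow adds to volcano smoke colors shaded by SplineThicken normals.
# Dynamic parameters control pan, glow size, width and erosion; opacity adds depth and near-camera fades, plus optional time-of-day scaling.
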Material /Game/Effects/Fort_Effects/Materials/Splines/M_Explosion_Arm_Spline
graph {
 properties {
  bUsedWithMeshParticles=true
  bUsedWithNiagaraMeshParticles=true
  blendMode=BLEND_Translucent
  pixelDepthOffsetMode=PDOM_Legacy
  shadingModel=MSM_Unlit
  twoSided=true
  usedShadingModels=Unlit
 }
 nodes [
  n1 = TextureCoordinate()
  LinearInterpolate(
   A=ScalarParameter() {parameterName=GlowBrightnessTip}
   B=ScalarParameter() {defaultValue=1, parameterName=GlowBrightnessBase}
   Alpha=ComponentMask(None=@n1) {g=true}
  )
  n2 = ParticleColor()
  n3 = ScalarParameter() {defaultValue=0.12, parameterName=GlowPosition}
  n4 = ScalarParameter() {defaultValue=1, parameterName=GlowRadius}
  n5 = DynamicParameter() {paramNames [PanSpeed,GlowSize,Width,Erode]}
  n6 = Power(
   Base=SphereMask(A=ComponentMask(None=@n1) {g=true}, B=@n3, Radius=Multiply(A=@n4, B=@n5.GlowSize)) {hardnessPercent=0}
   Exp=ScalarParameter() {defaultValue=2.5, group=Color, parameterName=GlowMaskPower}
  )
  n7 = AppendVector(A=Constant(), B=LinearInterpolate(Alpha=@n3) {constA=-1})
  n8 = Abs(None=ConstantBiasScale(None=Sine(None=@n3)))
  n9 = TextureCoordinate()
  n10 = AppendVector(A=Constant() {r=1}, B=ScalarParameter() {defaultValue=1, parameterName=Tile})
  n11 = Multiply(A=Multiply(A=Time(), B=ScalarParameter() {defaultValue=1, parameterName=PanSpeed}), B=@n5.PanSpeed)
  n12 = Panner(
   Coordinate=MaterialFunctionCall(XY=Multiply(A=@n9, B=@n10)) {materialFunction=&Swizzle}.YX
   Time=@n11
  ) {speedX=0.5}
  n13 = ScalarParameter() {defaultValue=0.5, parameterName=DetailTileX}
  n14 = ScalarParameter() {defaultValue=2, parameterName=DetailTileY}
  n15 = Panner(
   Coordinate=Multiply(A=AppendVector(A=@n13, B=@n14), B=@n9)
   Time=Multiply(A=ScalarParameter() {defaultValue=-0.5, parameterName=DetailPanMult}, B=@n11)
  ) {speedX=0.04, speedY=0.5}
  n16 = TextureSample(UVs=@n12) {texture=/Game/Effects/Fort_Effects/Gameplay/SmokeDustExplosions/Textures/T_RibbonSmokeMasked_04}
  n17 = Multiply(
   A=Multiply(
    A=VectorParameter() {parameterName=DetailNormalMult, defaultValue=<rgba 1 1 0 0>}.RGB
    B=TextureSample(UVs=@n15) {samplerType=SAMPLERTYPE_Normal, texture=/Game/Global/Textures/Default/Tileable_Patterns/T_StormCloudDetail01_N}.RGB
   )
   B=OneMinus(None=@n16.R)
  )
  n18 = Constant()
  n19 = Add(
   A=ShadingPathSwitch(Default=QualitySwitch(Default=@n17, Low=@n18), Mobile=@n18) {desc=Migrated from FeatureLevelSwitch}
   B=MaterialFunctionCall(XYZ=TextureSample(UVs=@n12) {samplerType=SAMPLERTYPE_Normal, texture=/Game/Effects/Fort_Effects/Gameplay/SmokeDustExplosions/Textures/T_RibbonSmokeMasked_01}.RGB) {materialFunction=&Swizzle}.YXZ
  )
  n20 = DotProduct(A=Normalize(VectorInput=AppendVector(A=@n7, B=@n8)), B=@n19)
  n21 = OneMinus(None=TextureSample(UVs=@n15) {samplerType=SAMPLERTYPE_LinearGrayscale, texture=/Game/Global/Textures/Default/Tileable_Patterns/T_StormCloudDetail01_M}.G)
  n22 = Constant() {r=0.2}
  n23 = ShadingPathSwitch(Default=QualitySwitch(Default=Multiply(A=@n21, B=@n21), Low=@n22), Mobile=@n22) {desc=Migrated from FeatureLevelSwitch}
  n24 = Add(
   A=Multiply(A=@n6, B=ScalarParameter() {defaultValue=0.5, group=Color, parameterName=GlowMaskAdd}) {constB=0.5}
   B=Multiply(A=Saturate(None=@n20), B=LinearInterpolate(A=@n23, Alpha=@n16.B) {constA=0.2}) {constB=0.5}
  ) {constB=0.2}
  n25 = Power(Base=@n24, Exp=ScalarParameter() {defaultValue=4, group=Color, parameterName=GlowMaskPower2}) {constExponent=3}
  n26 = Multiply(
   A=LinearInterpolate(
    A=VectorParameter() {group=Color, parameterName=GlowColor2, defaultValue=<rgba 1.5 0.015 0 0>}.RGB
    B=VectorParameter() {group=Color, parameterName=GlowColor1, defaultValue=<rgba 3 0.731 0 0>}.RGB
    Alpha=@n25
   )
   B=@n24
  )
  n27 = Multiply(A=ScalarParameter() {defaultValue=100, parameterName=BaseWidth}, B=@n5.Width)
  n28 = Multiply(A=@n27, B=ScalarParameter() {defaultValue=0.05, parameterName=WidthTipMult})
  n29 = Add(
   A=Multiply(A=@n2.RGB, B=Multiply(A=@n6, B=@n26))
   B=MaterialFunctionCall!(
    WorldSpaceNormal=MaterialFunctionCall(WidthBase=@n27, WidthTip=@n28, AdditionalNormal=@n19) {materialFunction=&SplineThicken}.Normal
    MaskForLowEnd=@n16.R
   ) {materialFunction=/Game/Athena/Prototype/Blueprints/White/FX/MF_Volcano_Smoke_Colors}.Result
  )
  MaterialFunctionCall(Emissive=LinearInterpolate(A=@n29, B=Multiply(A=@n2.RGB, B=@n29), Alpha=ScalarParameter() {group=Color, parameterName=Direct Particle Color Set})) {materialFunction=&MF_FX_EmissiveAdjustments}
  MaterialFunctionCall() {materialFunction=&DebugTimeSine}
  Saturate(None=@n25)
  n30 = ScalarParameter() {defaultValue=100, parameterName=DepthFadeDist}
  n31 = Multiply(
   A=MaterialFunctionCall(Distance=ScalarParameter() {defaultValue=200, parameterName=NearCamFadeDistance}) {materialFunction=&Near_Camera_Fade}.Result
   B=ShadingPathSwitch(Default=DepthFade(FadeDistance=@n30), Mobile=Constant() {r=1}) {desc=Migrated from FeatureLevelSwitch}
  )
  n32 = OneMinus(None=@n16.R)
  n33 = ComponentMask(None=TextureCoordinate()) {g=true}
  n34 = Add(A=Saturate(None=Divide(A=@n32) {constB=4}), B=Add(A=OneMinus(None=@n33), B=@n5.Erode))
  n35 = Clamp(None=@n5.Erode)
  n36 = LinearInterpolate(A=Multiply(A=@n35) {constB=-1}, B=Add(A=@n35, B=@n16.R), Alpha=@n35)
  n37 = Multiply(A=StaticSwitchParameter(True=Saturate(None=@n34), False=Saturate(None=@n36)) {parameterName=ErosionUp}, B=@n16.G)
  n38 = LinearInterpolate(Alpha=ComponentMask(None=TextureCoordinate()) {g=true}) {constA=4, constB=0}
  n39 = Multiply(A=@n2.A, B=Multiply(A=Multiply(A=@n37, B=@n31), B=Saturate(None=@n38)))
  n40 = ScalarParameter() {defaultValue=1, parameterName=OpacityMorning}
  n41 = ScalarParameter() {defaultValue=1, parameterName=OpacityDay}
  n42 = ScalarParameter() {defaultValue=1, parameterName=OpacityEvening}
  n43 = ScalarParameter() {defaultValue=1, parameterName=OpacityNight}
  StaticSwitchParameter(
   True=Multiply(
    A=@n39
    B=MaterialFunctionCall(Color_Morning=@n40, Color_Day=@n41, Color_Evening=@n42, Color_Night=@n43) {materialFunction=&Color_Time_of_Day}.Result
   )
   False=@n39
  ) {parameterName=Use Color_time_of_Day for Opacity}
 ]
}
