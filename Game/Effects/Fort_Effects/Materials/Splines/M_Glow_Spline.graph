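# Unlit, two-sided additive material for glowing spline effects on mesh particles.
# A soft mask across the U coordinate, sharpened by a power, combines with particle alpha and depth fade to blend from a tinted secondary color to particle color.
# That color feeds HDR-from-LDR and emissive adjustment functions.
# Vertex red also scales the mask, and SplineThicken supplies width-based world position offset, multiplied down to Z only.
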Material /Game/Effects/Fort_Effects/Materials/Splines/M_Glow_Spline
graph {
 properties {
  bUsedWithMeshParticles=true
  blendMode=BLEND_Additive
  pixelDepthOffsetMode=PDOM_Legacy
  shadingModel=MSM_Unlit
  twoSided=true
  usedShadingModels=Unlit
 }
 nodes [
  n1 = ParticleColor()
  n2 = Constant() {r=0.5}
  n3 = SphereMask(A=ComponentMask(None=TextureCoordinate()) {r=true}, B=@n2, Radius=@n2) {hardnessPercent=0}
  n4 = Multiply(A=@n1.A, B=Power(Base=@n3, Exp=ScalarParameter() {defaultValue=3, parameterName=Power}) {constExponent=5})
  n5 = DepthFade(FadeDistance=ScalarParameter() {defaultValue=400, parameterName=DepthFadeDistance})
  n6 = LinearInterpolate(
   A=Multiply(A=VectorParameter() {parameterName=SecondaryColor, defaultValue=<rgba 0.0427 0.056 0.1 0>}.RGB, B=@n1.RGB)
   B=@n1.RGB
   Alpha=Clamp(None=Multiply(A=@n4, B=@n5))
  )
  MaterialFunctionCall(Emissive=MaterialFunctionCall(LDR Input=Clamp(None=@n6), HDR Tint=@n6) {materialFunction=&DeriveHDRfromLDR}.RemappedRange) {materialFunction=&MF_FX_EmissiveAdjustments}
  Multiply(A=VertexColor().R, B=@n3)
  n7 = ScalarParameter() {defaultValue=200, parameterName=Width}
  Multiply(
   A=MaterialFunctionCall(WidthBase=@n7, WidthTip=@n7) {materialFunction=&SplineThicken}.WorldPositionOffset
   B=Constant3Vector() {constant=<rgba 0 0 1 0>}
  )
 ]
}
