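# Additive, unlit, two-sided laser spline material for Niagara mesh particles that draws without depth testing.
# SplineThicken takes a uniform width from a parameter scaled by dynamic WidthMult.
# A soft mask across U builds a powered inner core plus dim outer glow, tinted by particle color through emissive adjustments.
# A V mask scaled by object X scale and dynamic EndFadeHardness fades both ends, multiplied by particle alpha.
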Material /Game/Effects/Fort_Effects/Materials/Splines/M_Laser_Spline_NoDepthTest
graph {
 properties {
  bDisableDepthTest=true
  bUsedWithNiagaraMeshParticles=true
  blendMode=BLEND_Additive
  pixelDepthOffsetMode=PDOM_Legacy
  shadingModel=MSM_Unlit
  twoSided=true
  usedShadingModels=Unlit
 }
 nodes [
  n1 = DynamicParameter() {paramNames [WidthMult, EndFadeHardness, Param3, Param4]}
  n2 = Multiply(A=ScalarParameter() {defaultValue=10, parameterName=SplineWidth}, B=@n1.WidthMult)
  MaterialFunctionCall(WidthBase=@n2, WidthTip=@n2) {materialFunction=&SplineThicken}
  n3 = ParticleColor()
  n4 = MaterialFunctionCall(Float2=TextureCoordinate()) {materialFunction=&BreakOutFloat2Components}
  n5 = Constant() {r=0.5}
  n6 = SphereMask(A=@n4.R, B=@n5, Radius=@n5) {hardnessPercent=0}
  n7 = Multiply(
   A=Power(Base=@n6, Exp=ScalarParameter() {defaultValue=12, parameterName=InnerPower})
   B=ScalarParameter() {defaultValue=1.5, parameterName=InnerBrighness}
  )
  n8 = Multiply(A=@n6, B=ScalarParameter() {defaultValue=0.1, parameterName=OuterBrightness}) {constB=0.5}
  MaterialFunctionCall(Emissive=Multiply(A=@n3.RGB, B=Add(A=@n7, B=@n8))) {materialFunction=&MF_FX_EmissiveAdjustments}
  n9 = Multiply(
   A=SphereMask(A=@n4.G, B=@n5, Radius=@n5) {hardnessPercent=0}
   B=MaterialFunctionCall() {materialFunction=&ObjectScale}.Scale X
  )
  Multiply(A=@n3.A, B=Saturate(None=Multiply(A=@n9, B=@n1.EndFadeHardness)))
 ]
}
