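# Translucent, two-sided unlit material for light-pillar mesh particles that grow with camera distance so they read from afar.
# World position offset combines SplineThicken, using a particle-driven or scalar width, with expansion away from the actor origin, both driven by a clamped camera-distance ratio.
# Two gradient masks blend dark and light colors, tinted by particle color, and shape opacity with depth fade, particle alpha and an optional far camera fade.
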Material /Game/Effects/Fort_Effects/Materials/Splines/M_Light_Pillar_Distance_Scale
graph {
 properties {
  bUsedWithMeshParticles=true
  bUsedWithNiagaraMeshParticles=true
  blendMode=BLEND_Translucent
  pixelDepthOffsetMode=PDOM_Legacy
  shadingModel=MSM_Unlit
  twoSided=true
  usedShadingModels=Unlit
 }
 nodes [
  n1 = ActorPositionWS()
  n2 = Distance(A=@n1, B=CameraPositionWS())
  n3 = ScalarParameter() {defaultValue=2000, parameterName=DistanceOffset}
  n4 = Clamp(None=Divide(A=Subtract(A=@n2, B=@n3), B=ScalarParameter() {defaultValue=7000, parameterName=ScaleUpDistance}))
  n5 = Multiply(
   A=StaticSwitchParameter(
    True=DynamicParameter() {defaultValue=<rgba 100 1 1 1>, paramNames [SplineWidth,Param2,Param3,Param4]}.SplineWidth
    False=ScalarParameter() {defaultValue=50, parameterName=SplineWidth}
   ) {defaultValue=true, parameterName=UsedWithParticles}
   B=LinearInterpolate(B=ScalarParameter() {defaultValue=1, parameterName=MaxWidthMult}, Alpha=@n4) {constA=1, constB=3}
  )
  n6 = Multiply(A=ScalarParameter() {defaultValue=2, parameterName=MaxScaleIncrease}) {constB=-1}
  n7 = Subtract(A=@n1, B=WorldPosition().XYZ)
  Add(
   A=MaterialFunctionCall(WidthBase=@n5, WidthTip=@n5) {materialFunction=&SplineThicken}.WorldPositionOffset
   B=LinearInterpolate(B=Multiply(A=@n6, B=@n7) {constB=-1}, Alpha=@n4) {constB=0}
  )
  MaterialFunctionCall() {materialFunction=&DebugTimeSine}
  n8 = ParticleColor()
  n9 = VectorParameter() {parameterName=DarkColor, defaultValue=<rgba 0.3022 0.1424 0.3333 1>}
  n10 = LinearInterpolate(A=@n9.RGB, B=VectorParameter() {parameterName=LightColor, defaultValue=<rgba 1 1 1 0>}.RGB, Alpha=NamedRerouteUsage())
  n11 = TextureCoordinate()
  n12 = Multiply(
   A=TextureSample(UVs=Power(Base=@n11) {constExponent=0.6}) {texture=/Game/Global/Textures/Default/Masks_and_Gradients/T_Vert_Gradient_Soft}.R
   B=TextureSample(UVs=TextureCoordinate() {coordinateIndex=1}) {samplerType=SAMPLERTYPE_LinearColor, texture=/Game/Global/Textures/Default/Masks_and_Gradients/T_VertGradient}.R
  )
  MaterialFunctionCall(Emissive=Multiply(A=@n8.RGB, B=LinearInterpolate(A=@n9.RGB, B=@n10, Alpha=@n12))) {materialFunction=&MF_FX_EmissiveAdjustments}
  n13 = CameraPositionWS()
  n14 = WorldPosition()
  n15 = MaterialFunctionCall(
   Input=Length(None=Subtract(A=@n13, B=@n14.XYZ)) {is=class MaterialExpressionLength}
   Input Low=ScalarParameter() {defaultValue=1000, group=Camera Fade, parameterName=Camera Fade Far Start, sliderMax=50000}
   Input High=ScalarParameter() {defaultValue=3000, group=Camera Fade, parameterName=Camera Fade Far End, sliderMax=70000}
   Target Low=Constant() {r=1}
   Target High=Constant()
  ) {materialFunction=&RemapValueRange}
  NamedRerouteDeclaration(None=StaticSwitchParameter(True=Saturate(None=@n15.Result), False=Constant() {r=1}) {parameterName=Use Far Fade}) {name=Camera Fade Far, nodeColor=<rgba 1 0.0469 0 1>}
  n16 = ScalarParameter() {defaultValue=200, parameterName=DepthFadeDistance}
  n17 = Multiply(
   A=Multiply(A=@n8.A, B=DepthFade(Opacity=@n12, FadeDistance=@n16) {fadeDistanceDefault=200})
   B=ScalarParameter() {defaultValue=1, parameterName=AlphaIntensity}
  )
  n18 = ScalarParameter() {defaultValue=1, group=Camera Fade, parameterName=Camera Fade Far Alpha, sliderMax=1}
  n19 = NamedRerouteUsage()
  StaticSwitchParameter(True=Multiply(A=@n17, B=LinearInterpolate(A=@n18, Alpha=@n19)), False=@n17) {parameterName=Use Far Fade}
 ]
}
