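# Translucent unlit material for mesh-particle lightning arcs, thickened with SplineThicken and animated through world position offset.
# Layered sines on scrolling UVs form a normalized local-space wobble, sphere-masked to fade toward the arc's ends.
# A panning texture with randomized tiling blends edge and core colors into HDR emissive and drives end-faded opacity.
# Dynamic parameters vary width, wobble and timing per particle, and hide geometry by comparing vertex red against Random2.
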Material /Game/Effects/Fort_Effects/Materials/Splines/M_Lightning_Arc_WPO_Local_Stretch_RandomVertex
graph {
 properties {
  bUsedWithMeshParticles=true
  blendMode=BLEND_Translucent
  pixelDepthOffsetMode=PDOM_Legacy
  shadingModel=MSM_Unlit
  twoSided=true
  usedShadingModels=Unlit
 }
 nodes [
  n1 = DynamicParameter() {paramNames [Width,WobbleMult,Random1,Random2]}
  n2 = Multiply(A=ScalarParameter() {defaultValue=6, parameterName=Width}, B=@n1.Width)
  n3 = ScalarParameter() {parameterName=WidthTipMult}
  n4 = TextureCoordinate()
  n5 = ComponentMask(None=@n4) {g=true}
  n6 = Constant() {r=0.5}
  n7 = TextureCoordinate()
  n8 = Add(A=@n1.Random1, B=Time())
  n9 = Multiply(A=@n8, B=ScalarParameter() {defaultValue=3, parameterName=WobbleSpeed})
  n10 = ComponentMask(None=Add(A=@n7, B=Multiply(A=@n9) {constB=-0.6})) {g=true}
  n11 = Sine(None=@n10) {period=0.2}
  n12 = ComponentMask(None=Add(A=@n7, B=Multiply(A=@n9) {constB=-1})) {g=true}
  n13 = AppendVector(A=Constant(), B=Add(A=Multiply(A=@n11) {constB=0.2}, B=Sine(None=@n12) {period=0.8}))
  n14 = Add(A=Multiply(A=Sine(None=@n10) {period=0.23}) {constB=0.2}, B=Sine(None=@n12) {period=0.6})
  n15 = Normalize(VectorInput=Transform(None=AppendVector(A=@n13, B=@n14)) {transformSourceType=TRANSFORMSOURCE_Local})
  n16 = Multiply(A=ScalarParameter() {defaultValue=0.25, parameterName=Wobble}, B=@n1.WobbleMult)
  Add(
   A=MaterialFunctionCall(WidthBase=@n2, WidthTip=Multiply(A=@n2, B=@n3) {constB=5}, UVs for Thickness=Power(Base=@n4) {constExponent=0.6}) {materialFunction=&SplineThicken}.WorldPositionOffset
   B=Multiply(A=SphereMask(A=@n5, B=@n6, Radius=@n6) {hardnessPercent=50}, B=Multiply(A=@n15, B=@n16))
  )
  n17 = ParticleColor()
  n18 = TextureCoordinate() {vTiling=0.5}
  n19 = AppendVector(A=Constant() {r=1}, B=LinearInterpolate(Alpha=@n1.Random1) {constA=0.5, constB=1.2})
  n20 = ScalarParameter() {defaultValue=1, parameterName=PanSpeed}
  n21 = TextureSampleParameter2D(UVs=Panner(Coordinate=Multiply(A=@n18, B=@n19), Time=Multiply(A=@n8, B=@n20)) {speedY=-1}) {parameterName=Texture, samplerType=SAMPLERTYPE_LinearColor, texture=/Game/Effects/Fort_Effects/Textures/Tile/T_AMR_MuzzleRing_01}
  n22 = Multiply(
   A=@n17.RGB
   B=LinearInterpolate(
    A=VectorParameter() {parameterName=EdgeColor, defaultValue=<rgba 0.2226 0.2425 0.5 0>}.RGB
    B=VectorParameter() {parameterName=Color, defaultValue=<rgba 0.6214 0.9501 1 0>}.RGB
    Alpha=@n21.G
   )
  )
  MaterialFunctionCall(Emissive=MaterialFunctionCall(LDR Input=Clamp(None=@n22), HDR Tint=@n22, HDR Intensity Multiplier=ScalarParameter() {defaultValue=20, parameterName=HDRIntensity}) {materialFunction=&DeriveHDRfromLDR}.RemappedRange) {materialFunction=&MF_FX_EmissiveAdjustments}
  Multiply()
  n23 = ComponentMask(None=TextureCoordinate()) {g=true}
  n24 = OneMinus(None=@n23)
  n25 = Add(A=@n21.R, B=Multiply(A=@n21.B) {constB=0.2})
  n26 = Constant()
  Multiply(
   A=Multiply(A=Multiply(A=@n23, B=@n24), B=Multiply(A=@n25, B=@n17.A))
   B=If(
    A=VertexColor().R
    B=DynamicParameter() {paramNames [Width,WobbleMult,Random1,Random2]}.Random2
    "A > B"=Constant() {r=1}
    "A == B"=@n26
    "A < B"=@n26
   )
  )
 ]
}
