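# Translucent unlit mesh-particle material for a lightning arc that stretches toward a target point.
# World position offset sums SplineThicken width, time-driven sine wobble masked along V, and a V-weighted pull toward a randomly offset actor position or Niagara parameter.
# A panning plasma texture tinted by particle color drives HDR emissive, with optional intersection boost.
# Opacity combines texture channels, particle alpha, optional end and ramp masks, and quality-switched depth fade.
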Material /Game/Effects/Fort_Effects/Materials/Splines/M_Lightning_Arc_WPO_Stretch
graph {
 properties {
  bUsedWithMeshParticles=true
  bUsedWithNiagaraMeshParticles=true
  blendMode=BLEND_Translucent
  pixelDepthOffsetMode=PDOM_Legacy
  shadingModel=MSM_Unlit
  twoSided=true
  usedShadingModels=Unlit
 }
 nodes [
  n1 = DynamicParameter() {paramNames [Width,WobbleMult,Random1,Random2]}
  n2 = Multiply(A=ScalarParameter() {defaultValue=6, parameterName=Width}, B=@n1.Width)
  n3 = ScalarParameter() {parameterName=WidthTipMult}
  n4 = TextureCoordinate()
  n5 = Constant() {r=0.5}
  n6 = SphereMask(A=ComponentMask(None=@n4) {g=true}, B=@n5, Radius=@n5) {hardnessPercent=50}
  n7 = TextureCoordinate()
  n8 = Add(A=@n1.Random2, B=Time())
  n9 = Multiply(A=@n8, B=ScalarParameter() {defaultValue=3, parameterName=WobbleSpeed})
  n10 = ComponentMask(None=Add(A=@n7, B=Multiply(A=@n9) {constB=-0.6})) {g=true}
  n11 = Multiply(A=Sine(None=@n10) {period=0.2}) {constB=0.2}
  n12 = ComponentMask(None=Add(A=@n7, B=Multiply(A=@n9) {constB=-1})) {g=true}
  n13 = Sine(None=@n12) {period=0.8}
  n14 = Sine(None=@n10) {period=0.23}
  n15 = AppendVector(A=AppendVector(A=Constant(), B=Add(A=@n11, B=@n13)), B=Add(A=Multiply(A=@n14) {constB=0.2}, B=Sine(None=@n12) {period=0.6}))
  n16 = Multiply(
   A=Normalize(VectorInput=Transform(None=@n15) {transformSourceType=TRANSFORMSOURCE_Local})
   B=Multiply(A=ScalarParameter() {defaultValue=5, parameterName=Wobble}, B=@n1.WobbleMult)
  )
  n17 = ScalarParameter() {defaultValue=1, parameterName=StetchWPOMult}
  n18 = DynamicParameter() {desc=Index 2 parameter for Niagara, parameterIndex=1, defaultValue=<rgba 0 0 0 0>, paramNames [PosX,PosY,PosZ,Param4]}
  n19 = Multiply(
   A=Subtract(A=AppendVector(A=@n1.Random1, B=@n1.Random1)) {constB=0.5}
   B=ScalarParameter() {defaultValue=30, parameterName=ActorPositionOffset}
  ) {constB=10}
  n20 = Constant()
  n21 = Add(
   A=LinearInterpolate(
    A=ActorPositionWS()
    B=MaterialFunctionCall(X=@n18.PosX, Y=@n18.PosX, Z=@n18.PosX) {materialFunction=&MakeFloat3}.Result
    Alpha=ScalarParameter() {parameterName=UseNiagaraPositionParam}
   )
   B=Add(A=AppendVector(A=@n19, B=@n20), B=VectorParameter() {parameterName=ActorPositionAdd}.RGB)
  )
  n22 = WorldPosition()
  n23 = ComponentMask(None=TextureCoordinate()) {g=true}
  n24 = Multiply(A=Multiply(A=Subtract(A=@n21, B=@n22.XYZ), B=OneMinus(None=@n23)), B=Clamp(None=LinearInterpolate() {constA=2, constB=0}, Min=@n18.Param4))
  Add(
   A=MaterialFunctionCall(WidthBase=@n2, WidthTip=Multiply(A=@n2, B=@n3) {constB=5}, UVs for Thickness=Power(Base=@n4) {constExponent=0.6}) {materialFunction=&SplineThicken}.WorldPositionOffset
   B=Add(A=Multiply(A=@n6, B=@n16), B=Multiply(A=@n17, B=@n24))
  )
  n25 = ParticleColor()
  n26 = TextureCoordinate() {vTiling=0.5}
  n27 = AppendVector(A=Constant() {r=1}, B=LinearInterpolate(Alpha=@n1.Random1) {constA=0.5, constB=1.2})
  n28 = ScalarParameter() {defaultValue=1, parameterName=PanSpeed}
  n29 = TextureSample(UVs=Panner(Coordinate=Multiply(A=@n26, B=@n27), Time=Multiply(A=@n8, B=@n28)) {speedY=-1}) {texture=/Game/Effects/Fort_Effects/Textures/Tile/T_Lightning_Plasma_Soft}
  n30 = Multiply(
   A=@n25.RGB
   B=LinearInterpolate(
    A=VectorParameter() {parameterName=EdgeColor, defaultValue=<rgba 0.2226 0.2425 0.5 0>}.RGB
    B=VectorParameter() {parameterName=Color, defaultValue=<rgba 0.6214 0.9501 1 0>}.RGB
    Alpha=@n29.G
   )
  )
  n31 = MaterialFunctionCall(LDR Input=Clamp(None=@n30), HDR Tint=@n30, HDR Intensity Multiplier=ScalarParameter() {defaultValue=20, parameterName=HDRIntensity}) {materialFunction=&DeriveHDRfromLDR}
  n32 = DepthFade(FadeDistance=ScalarParameter() {defaultValue=100, parameterName=IntesectingThickness})
  n33 = LinearInterpolate(
   A=Multiply(A=@n31.RemappedRange, B=ScalarParameter() {defaultValue=2, parameterName=IntesectingEmissiveBoost})
   B=@n31.RemappedRange
   Alpha=@n32
  )
  MaterialFunctionCall(Emissive=StaticSwitchParameter(
    True=MaterialFunctionCall!(Default=@n33, Low/Mobile=@n31.RemappedRange) {materialFunction=/Game/Effects/Fort_Effects/Materials/_Functions/MF_LowMobile_PathingSwitch}.Result
    False=@n31.RemappedRange
   ) {parameterName=MeshIntersectingColorMultiply}) {materialFunction=&MF_FX_EmissiveAdjustments}
  n34 = TextureCoordinate()
  n35 = Power(
   Base=Sine(None=ComponentMask(None=@n34) {g=true}) {period=2}
   Exp=ScalarParameter() {defaultValue=1, group=Opacity, parameterName=Ramp Mask Power}
  )
  NamedRerouteDeclaration(None=If(A=ScalarParameter() {group=Opacity, parameterName=Enable Ramp Mask}, "A > B"=@n35, "A == B"=@n35, "A < B"=Constant() {r=1}) {constB=0.99}) {name=Ramp Fade, nodeColor=<rgba 0 1 0.3203 1>}
  ParticleRelativeTime()
  Saturate(None=Power(Base=@n32))
  n36 = ComponentMask(None=TextureCoordinate()) {g=true}
  n37 = StaticSwitchParameter(True=OneMinus(None=@n36), False=@n36) {parameterName=AlphaMasks_Invert}
  n38 = ScalarParameter() {defaultValue=2, parameterName=AlphaMask_Power}
  n39 = Saturate(None=Multiply(A=Power(Base=@n37, Exp=@n38), B=ScalarParameter() {defaultValue=3, parameterName=AlphaMask_Multiply}) {constB=3})
  n40 = Multiply(A=Add(A=@n29.R, B=Multiply(A=@n29.B) {constB=0.2}), B=@n25.A)
  n41 = Multiply(A=StaticSwitchParameter(True=Multiply(A=@n39, B=@n40), False=@n40) {parameterName=AlphaMasks}, B=NamedRerouteUsage())
  n42 = ScalarParameter() {group=Opacity, parameterName=Depth Fade}
  ShadingPathSwitch(Default=QualitySwitch(Default=DepthFade(Opacity=@n41, FadeDistance=@n42), Low=@n41), Mobile=@n41) {desc=Migrated from FeatureLevelSwitch}
 ]
}
