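# Additive, unlit spline material for water splash effects on mesh particles, faking sunlight in emissive.
# SplineThicken normals, widened by a dynamic parameter, blend with a splash normal map, feeding a sun/moon-direction specular highlight tinted by time of day; mobile uses constant lighting.
# Random 2x2 sub-UV splash textures blend channels by dynamic parameter, tinted by base and particle color.
# Opacity combines texture, particle alpha and near-camera/depth fades, skipped on low quality.
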Material /Game/Effects/Fort_Effects/Materials/Splines/M_Splash_Spline_FakeLit
graph {
 properties {
  bTangentSpaceNormal=false
  bUsedWithMeshParticles=true
  bUsedWithNiagaraMeshParticles=true
  blendMode=BLEND_Additive
  opacityMaskClipValue=0.933
  pixelDepthOffsetMode=PDOM_Legacy
  refractionMethod=RM_IndexOfRefraction
  shadingModel=MSM_Unlit
  translucencyPass=MTP_BeforeDOF
  twoSided=true
  usedShadingModels=Unlit
 }
 nodes [
  Constant() {r=1}
  n1 = DynamicParameter() {defaultValue=<rgba 0 1 1 1>, paramNames [TextureBlend,PanSpeed,Random,Width]}
  n2 = Multiply(A=@n1.Width, B=ScalarParameter() {defaultValue=40, parameterName=Width})
  n3 = ScalarParameter() {defaultValue=0.5, parameterName=WidthBaseMult}
  n4 = Constant() {r=2}
  n5 = MaterialFunctionCall!(UseTexCoord 1, Rows=@n4, Columns=@n4, Particle Random Value=@n1.Random) {materialFunction=/Game/Effects/Fort_Effects/Materials/_Functions/MF_RandomSubUV_GPU_Sprites}
  MaterialFunctionCall(
   BaseNormal=MaterialFunctionCall(WidthBase=Multiply(A=@n2, B=@n3) {constB=0.5}, WidthTip=@n2) {materialFunction=&SplineThicken}.Normal
   AdditionalNormal=Transform(None=TextureSample(UVs=@n5.Result) {samplerType=SAMPLERTYPE_Normal, texture=/Game/Effects/Fort_Effects/Textures/Water/T_waterSplash_N}.RGB)
  ) {materialFunction=&BlendAngleCorrectedNormals}
  n6 = MaterialFunctionCall() {materialFunction=&Color_Time_of_Day}
  n7 = LinearInterpolate(B=@n6.Result) {constA=1, constAlpha=0.25}
  n8 = LinearInterpolate(B=@n6.Result) {constA=1, constAlpha=0.6}
  n9 = CollectionParameter() {
   collection=&FortniteMaterialParameters
   parameterName=SunAndMoonModelDirectionalVector
  }
  n10 = DotProduct(A=ComponentMask(None=@n9) {b=true, g=true, r=true}, B=PixelNormalWS())
  n11 = ScalarParameter() {defaultValue=30, parameterName=Spec Highlight Power}
  n12 = Multiply(A=Clamp(None=Power(Base=@n10, Exp=@n11) {constExponent=20})) {constB=3}
  n13 = ShadingPathSwitch(Default=Add(A=Multiply(A=@n7) {constB=0.2}, B=Multiply(A=@n8, B=@n12)), Mobile=Constant() {r=0.8}) {desc=Migrated from FeatureLevelSwitch}
  n14 = TextureSampleParameter2D(UVs=@n5.Result) {parameterName=FlashTexture, samplerType=SAMPLERTYPE_LinearColor, texture=/Game/Effects/Fort_Effects/Textures/Water/T_waterSplash}
  n15 = TextureSampleParameter2D(UVs=@n5.Result) {parameterName=FlashTexture, samplerType=SAMPLERTYPE_LinearColor, texture=&T_waterSplash}
  n16 = Multiply(
   A=LinearInterpolate(A=@n14.R, B=@n15.B, Alpha=@n1.TextureBlend)
   B=VectorParameter() {parameterName=BaseColor, defaultValue=<rgba 0.4437 0.4684 0.5 0>}.RGB
  )
  n17 = ParticleColor()
  MaterialFunctionCall(Emissive=Multiply(A=Multiply(A=@n13, B=@n16), B=@n17.RGB)) {materialFunction=&MF_FX_EmissiveAdjustments}
  n18 = DepthFade(
   Opacity=MaterialFunctionCall(Distance=ScalarParameter() {defaultValue=200, parameterName=NearFadeDist}) {materialFunction=&Near_Camera_Fade}.Result
   FadeDistance=ScalarParameter() {defaultValue=40, parameterName=DepthFadeDist}
  )
  Multiply(A=Multiply(A=@n14.R, B=@n17.A), B=QualitySwitch(Default=@n18, Low=Constant() {r=1}, High=@n18, Medium=@n18))
  TextureCoordinate() {uTiling=0.5, vTiling=0.95}
 ]
}
