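# Unlit, additive, two-sided material for spline-thickened light beam effects on mesh particles, built procedurally without textures.
# A UV-based tiled soft mask blends dark and particle-tinted bright color for emissive and shapes opacity, with angle and depth fades skipped on mobile.
# SplineThicken takes base and tip widths from dynamic parameters, optionally vertex red.
# Opacity also gets optional sine-product flicker, with CPU/GPU particle switches, and a game-time fade.
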Material /Game/Effects/Fort_Effects/Materials/Splines/M_Spline_LightBeam
graph {
 properties {
  bUsedWithMeshParticles=true
  bUsedWithNiagaraMeshParticles=true
  blendMode=BLEND_Additive
  pixelDepthOffsetMode=PDOM_Legacy
  shadingModel=MSM_Unlit
  twoSided=true
  usedShadingModels=Unlit
 }
 nodes [
  n1 = ParticleColor()
  n2 = Multiply(A=@n1.RGB, B=VectorParameter() {parameterName=DarkColor, defaultValue=<rgba 0.1935 0.1649 0.4167 0>}.RGB)
  n3 = OneMinus(None=DepthFade(Opacity=ScalarParameter() {defaultValue=4, parameterName=BrightFade}))
  n4 = ComponentMask(None=TextureCoordinate()) {g=true}
  n5 = TextureCoordinate()
  n6 = ComponentMask(None=@n5) {g=true}
  n7 = Add(A=Multiply(A=Multiply(A=@n3, B=@n4)) {constB=2}, B=@n6)
  n8 = Add(A=Multiply(A=Power(Base=@n7) {constExponent=4}, B=@n1.RGB), B=@n1.RGB)
  n9 = LinearInterpolate(A=ScalarParameter() {defaultValue=3, parameterName=BottomFadeSize}, Alpha=@n6) {constA=1, constB=0}
  n10 = Fmod(A=Multiply(A=ComponentMask(None=@n5) {r=true}, B=ScalarParameter() {defaultValue=1, parameterName=BeamTile}), B=Constant() {r=1})
  n11 = Constant() {r=0.5}
  n12 = Multiply(
   A=LinearInterpolate(B=Saturate(None=@n9), Alpha=ScalarParameter() {parameterName=UseBottomFade}) {constA=1}
   B=Multiply(A=@n6, B=SphereMask(A=@n10, B=@n11, Radius=@n11) {hardnessPercent=0})
  )
  MaterialFunctionCall(Emissive=Multiply(A=LinearInterpolate(A=@n2, B=@n8, Alpha=@n12), B=ScalarParameter() {defaultValue=1, parameterName=emissive mult})) {materialFunction=&MF_FX_EmissiveAdjustments}
  n13 = ParticleRandom()
  n14 = ScalarParameter() {defaultValue=1, group=Flicker, parameterName=FlickerSpeed}
  n15 = DynamicParameter() {paramNames [WidthBaseMult, WidthTipMult, FlickerFade, FlickerSpeed]}
  n16 = Multiply(
   A=Add(A=Time(), B=VertexColor().R)
   B=StaticSwitchParameter(
    True=LinearInterpolate(Alpha=@n13) {constA=1.2, constB=0.8}
    False=StaticSwitchParameter(True=@n15.FlickerSpeed, False=@n14) {parameterName=Particles?}
   ) {parameterName=GPU Particles}
  )
  n17 = Sine(None=@n16) {period=0.266}
  n18 = Multiply(A=Sine(None=@n16) {period=0.155}, B=Sine(None=@n16) {period=0.2})
  n19 = ScalarParameter() {defaultValue=0.4, desc="0 = full flicker amount", group=Flicker, parameterName=GPUFlickerFadeMin}
  n20 = ScalarParameter() {defaultValue=1, desc="1 = no flicker", group=Flicker, parameterName=GPUFlickerFadeMax, sortPriority=1}
  n21 = ParticleRandom()
  n22 = ScalarParameter() {defaultValue=1, group=Flicker, parameterName=FlickerFade}
  n23 = LinearInterpolate(
   A=Clamp(None=Multiply(A=@n17, B=@n18)) {maxDefault=3}
   Alpha=StaticSwitchParameter(
    True=LinearInterpolate(A=@n19, B=@n20, Alpha=@n21) {constA=0.5}
    False=StaticSwitchParameter(True=@n15.FlickerFade, False=@n22) {parameterName=Particles?}
   ) {parameterName=GPU Particles}
  ) {constA=1}
  n24 = Constant() {r=1}
  n25 = DynamicParameter() {paramNames [WidthBaseMult, WidthTipMult, FlickerFade, FlickerSpeed]}
  n26 = ScalarParameter() {defaultValue=5, parameterName=WidthBase}
  n27 = Multiply(A=@n25.WidthTipMult, B=ScalarParameter() {defaultValue=100, parameterName=WidthTip})
  n28 = Multiply(A=VertexColor().R, B=@n27)
  n29 = DotProduct(
   A=CameraVectorWS()
   B=MaterialFunctionCall(
    WidthBase=Multiply(A=@n25.WidthBaseMult, B=@n26)
    WidthTip=StaticSwitchParameter(True=@n28, False=@n27) {defaultValue=true, parameterName=VertexColorTipWidthOffset}
   ) {materialFunction=&SplineThicken}.Normal
  )
  n30 = ScalarParameter() {defaultValue=0.5, parameterName=AngleFade}
  n31 = StaticSwitchParameter(
   True=Clamp(None=Power(Base=@n29, Exp=@n30))
   False=MaterialFunctionCall("FadeSensitivity (optional)"=@n30) {materialFunction=&FadeBasedOnViewAngle}.Result
  ) {parameterName=Use WPO Adjusted Normals for Angle Fade}
  n32 = TextureCoordinate()
  n33 = DepthFade(FadeDistance=LinearInterpolate(
    A=ScalarParameter() {defaultValue=80, parameterName=DepthTipDist}
    B=ScalarParameter() {parameterName=DepthBaseDist}
    Alpha=ComponentMask(None=@n32) {g=true}
   ))
  n34 = Multiply(
   A=Power(Base=@n12, Exp=ScalarParameter() {defaultValue=1, parameterName=OpacityPower})
   B=ShadingPathSwitch(Default=Multiply(A=@n31, B=@n33), Mobile=Constant() {r=1}) {desc=Migrated from FeatureLevelSwitch}
  )
  Multiply(A=Multiply(A=StaticSwitchParameter(True=@n23, False=@n24) {parameterName=UseFlicker}, B=Multiply(A=@n1.A, B=@n34)), B=NamedRerouteUsage())
  n35 = Time()
  n36 = ScalarParameter() {group=Game Time Fade, parameterName=FadeGameTime}
  n37 = Saturate(None=Divide(A=Subtract(A=@n35, B=@n36), B=ScalarParameter() {defaultValue=2, group=Game Time Fade, parameterName=FadeDuration}))
  NamedRerouteDeclaration(None=LinearInterpolate(A=OneMinus(None=@n37), B=@n37, Alpha=ScalarParameter() {defaultValue=1, group=Game Time Fade, parameterName=FadeDirection})) {name=Game Time Fade, nodeColor=<rgba 0 1 0.1562 1>}
  VertexNormalWS()
 ]
}
