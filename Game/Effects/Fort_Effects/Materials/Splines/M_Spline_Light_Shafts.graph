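# Additive, unlit, two-sided material that draws light shafts on spline-thickened meshes or Niagara mesh particles.
# World position offset combines SplineThicken with a stretch away from an offset source near the pivot, driven by dynamic parameters for width, offset and length.
# Panning noise, UV gradients and view-angle fade shape color from parameters or particle color, boosted by a time-of-day night multiplier.
# Particle alpha, depth fade and game-time fade control visibility.
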Material /Game/Effects/Fort_Effects/Materials/Splines/M_Spline_Light_Shafts
graph {
 properties {
  bUsedWithNiagaraMeshParticles=true
  blendMode=BLEND_Additive
  pixelDepthOffsetMode=PDOM_Legacy
  shadingModel=MSM_Unlit
  twoSided=true
  usedShadingModels=Unlit
 }
 nodes [
  n1 = VectorParameter() {parameterName=NoiseTile, defaultValue=<rgba 1.4 0.25 0 0>}
  n2 = Add(A=Multiply(A=TextureCoordinate(), B=AppendVector(A=@n1.R, B=@n1.R)), B=TextureCoordinate() {coordinateIndex=1})
  n3 = DynamicParameter() {paramNames [Width, Offset, Length, DepthFadeDist]}
  Add(A=@n2, B=Multiply(A=Constant2Vector() {g=0.0001, r=0.002}, B=@n3.Offset))
  MaterialFunctionCall() {materialFunction=&DebugTimeSine}
  n4 = SphereMask(A=TextureCoordinate() {coordinateIndex=1}, B=Constant() {r=0.5}, Radius=ScalarParameter() {defaultValue=1, parameterName=SphereRadius}) {attenuationRadius=1, hardnessPercent=0}
  n5 = Multiply(A=@n4, B=ScalarParameter() {defaultValue=3400, parameterName=Length})
  n6 = TextureCoordinate()
  n7 = ComponentMask(None=@n6) {g=true}
  n8 = LinearInterpolate(A=Multiply(A=ScalarParameter() {defaultValue=5, parameterName=TopFade}, B=@n5), Alpha=@n7) {constA=15, constB=0}
  n9 = ComponentMask(None=@n6) {r=true}
  n10 = Constant() {r=0.5}
  n11 = Multiply(
   A=Multiply(A=Saturate(None=@n8), B=Power(Base=@n7) {constExponent=2.5})
   B=Power(Base=SphereMask(A=@n9, B=@n10, Radius=@n10) {hardnessPercent=0})
  )
  n12 = OneMinus(None=ComponentMask(None=TextureCoordinate()) {g=true})
  n13 = LinearInterpolate(Alpha=@n12) {constA=15, constB=0}
  n14 = ScalarParameter() {defaultValue=2.5, group=BottomFade, parameterName=BottomFadeExp}
  n15 = Multiply(A=Multiply(A=Saturate(None=@n13), B=Power(Base=@n12, Exp=@n14)), B=@n11)
  n16 = ScalarParameter() {group=BottomFade, parameterName=UseBottomFade}
  n17 = ScalarParameter() {defaultValue=0.1, parameterName=MinNoise}
  n18 = Time()
  n19 = ScalarParameter() {defaultValue=0.5, parameterName=PanSpeed}
  n20 = TextureSample(UVs=Panner(Coordinate=@n2, Time=Multiply(A=@n18, B=@n19)) {speedX=0.1, speedY=1}) {texture=&T_Smooth_tile}
  n21 = Multiply(
   A=Multiply(A=LinearInterpolate(A=@n11, B=@n15, Alpha=@n16), B=LinearInterpolate(A=@n17, Alpha=@n20.G))
   B=MaterialFunctionCall("FadeSensitivity (optional)"=ScalarParameter() {defaultValue=1.5, parameterName=AngleFade}) {materialFunction=&FadeBasedOnViewAngle}.Result
  )
  n22 = Multiply(
   A=Multiply(
    A=ScalarParameter() {defaultValue=1, group=Emissive, parameterName=Emissive Multiplier}
    B=VectorParameter() {parameterName=Color, defaultValue=<rgba 1 0 1 1>}.RGB
   )
   B=@n21
  )
  n23 = ParticleColor()
  n24 = Multiply(A=@n23.RGB, B=@n21)
  n25 = CollectionParameter() {
   collection=&FortniteMaterialParameters
   parameterName=FortniteActiveTimeOfDay
  }
  Multiply(
   A=MaterialFunctionCall(Emissive=StaticSwitchParameter(True=@n22, False=@n24) {parameterName=UseOnStaticMesh?}) {materialFunction=&MF_FX_EmissiveAdjustments}.Result
   B=LinearInterpolate(B=ScalarParameter() {defaultValue=1, group=Emissive, parameterName=Emissive Night Multiplier}, Alpha=ComponentMask(None=@n25) {a=true}) {constA=1}
  )
  n26 = ScalarParameter() {defaultValue=1, parameterName=DepthFadeDistance}
  Multiply(A=DepthFade(Opacity=@n23.A, FadeDistance=Multiply(A=@n3.DepthFadeDist, B=@n26)), B=NamedRerouteUsage())
  n27 = Time()
  n28 = ScalarParameter() {group=Game Time Fade, parameterName=FadeGameTime}
  n29 = Saturate(None=Divide(A=Subtract(A=@n27, B=@n28), B=ScalarParameter() {defaultValue=2, group=Game Time Fade, parameterName=FadeDuration}))
  NamedRerouteDeclaration(None=LinearInterpolate(A=OneMinus(None=@n29), B=@n29, Alpha=ScalarParameter() {defaultValue=1, group=Game Time Fade, parameterName=FadeDirection})) {name=Game Time Fade, nodeColor=<rgba 0.1484 0 1 1>}
  n30 = Multiply(A=Multiply(A=@n3.Width, B=ScalarParameter() {defaultValue=100, parameterName=SplineWidth}), B=@n4)
  n31 = Multiply(A=@n30, B=ScalarParameter() {defaultValue=2, parameterName=SplineWidthTipMult})
  n32 = WorldPosition()
  n33 = ScalarParameter() {defaultValue=40, parameterName=LightSourceOffset}
  n34 = Transform(None=Constant3Vector() {constant=<rgba 0 0 1 0>}) {transformSourceType=TRANSFORMSOURCE_Local}
  n35 = Add(
   A=Multiply(A=Multiply(A=@n3.Offset, B=@n33), B=Normalize(VectorInput=@n34))
   B=MaterialFunctionCall() {materialFunction=&ObjectPivotPoint}.Object Pivot Location
  )
  n36 = Multiply(A=Multiply(A=@n5, B=@n3.Length), B=Normalize(VectorInput=Subtract(A=@n32.XYZ, B=@n35)))
  n37 = ComponentMask(None=TextureCoordinate()) {g=true}
  Reroute(None=Add(
    A=MaterialFunctionCall(WidthBase=@n30, WidthTip=@n31) {materialFunction=&SplineThicken}.WorldPositionOffset
    B=LinearInterpolate(A=@n36, Alpha=@n37) {constB=0}
   ))
 ]
}
