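# Additive, unlit, two-sided material for mesh-particle wind-trail splines, judging by its name and smoke-trail textures.
# Panned smoke textures tinted by particle color drive erosion and emissive, shaped by sphere masks, depth fade, and optional end fades and fresnel.
# World position offset widens strips through SplineThicken or tangent/binormal expansion, shrinking near the camera.
# Dynamic parameters control width, a sliding V-axis mask and, via static switch, erosion.
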Material /Game/Effects/Fort_Effects/Materials/Splines/M_Wind_Spline_Travel
graph {
 properties {
  bUsedWithMeshParticles=true
  bUsedWithNiagaraMeshParticles=true
  blendMode=BLEND_Additive
  pixelDepthOffsetMode=PDOM_Legacy
  shadingModel=MSM_Unlit
  twoSided=true
  usedShadingModels=Unlit
 }
 nodes [
  n1 = ScalarParameter() {defaultValue=1, parameterName=Use Top Sphere Mask}
  n2 = SphereMask(A=TextureCoordinate(), B=Constant2Vector() {r=0.5}) {attenuationRadius=0.5, hardnessPercent=0}
  n3 = Power(Base=@n2) {constExponent=4}
  n4 = Multiply(A=@n1, B=@n3)
  n5 = ScalarParameter() {defaultValue=-3, parameterName=Edge-Trailing_GradMult}
  n6 = StaticSwitchParameter(
   True=DynamicParameter() {paramNames [EndErosion,Param2,PanAdd,SplineWidth]}.EndErosion
   False=ScalarParameter() {defaultValue=1, parameterName=Edge-Leading_GradMult}
  ) {parameterName=DynamicParameter}
  n7 = ParticleColor()
  n8 = ComponentMask(None=TextureCoordinate() {coordinateIndex=1}) {r=true}
  n9 = ScalarParameter() {defaultValue=2.5, parameterName=X-Tiling}
  n10 = Constant() {r=1}
  n11 = Panner(
   Coordinate=Multiply(A=TextureCoordinate() {coordinateIndex=1}, B=AppendVector(A=@n9, B=@n10))
   Time=Multiply(A=Time(), B=ScalarParameter() {defaultValue=3, parameterName=PanSpeed})
  ) {speedX=0.5}
  n12 = TextureSampleParameter2D(UVs=@n11) {parameterName=Mask, samplerType=SAMPLERTYPE_Grayscale, texture=&T_Smoke_Trail_Soft}
  n13 = Add(A=Multiply(A=@n7.A, B=Add(A=@n8) {constB=0.1}), B=@n12.R)
  n14 = Clamp(None=Multiply(A=LinearInterpolate(A=@n5, B=@n6, Alpha=@n13) {constA=-3}, B=@n12.R))
  n15 = ComponentMask(None=TextureCoordinate()) {g=true}
  n16 = VectorParameter() {parameterName=Sphere Masks Options, channelNames {a=Hardness B, b=Radius B, g=Hardness A, r=Radius A}, defaultValue=<rgba 0.5 0.9 1 0.5>}
  n17 = SphereMask(A=@n15, B=Constant() {r=0.5}, Radius=@n16.Radius A, Hardness=@n16.Radius A) {attenuationRadius=0.5, hardnessPercent=90}
  n18 = DynamicParameter() {paramNames [EndErosion,Param2,PanAdd,SplineWidth]}
  n19 = SphereMask(A=@n15, B=LinearInterpolate(Alpha=@n18.PanAdd) {constA=-1, constB=0.25}, Radius=@n16.Radius B, Hardness=@n16.Radius B) {attenuationRadius=1, hardnessPercent=50}
  n20 = Multiply(A=@n7.A, B=Multiply(A=Add(A=@n4, B=@n14), B=Multiply(A=@n17, B=@n19) {constA=1}))
  n21 = ScalarParameter() {defaultValue=20, parameterName=DepthFadeDist}
  n22 = Multiply(
   A=@n20
   B=ShadingPathSwitch(Default=DepthFade(Opacity=@n20, FadeDistance=@n21) {fadeDistanceDefault=55}, Mobile=Constant() {r=1}) {desc=Migrated from FeatureLevelSwitch}
  )
  n23 = TextureCoordinate()
  n24 = ScalarParameter() {defaultValue=1, parameterName=EndFadeMult}
  n25 = Clamp(None=Multiply(A=ComponentMask(None=@n23) {r=true}, B=@n24))
  n26 = LinearInterpolate(A=@n22, B=Multiply(A=@n22, B=Power(Base=@n25, Exp=@n24)), Alpha=ScalarParameter() {parameterName=UseEndFade?})
  n27 = ComponentMask(None=TextureCoordinate()) {r=true}
  n28 = ScalarParameter() {defaultValue=1, parameterName=StartFadeMult}
  n29 = Clamp(None=Multiply(A=OneMinus(None=@n27), B=@n28))
  n30 = LinearInterpolate(A=@n26, B=Multiply(A=@n26, B=Power(Base=@n29, Exp=@n28)), Alpha=ScalarParameter() {parameterName=UseStartFade?})
  n31 = DotProduct(A=CameraVectorWS(), B=VertexNormalWS())
  n32 = LinearInterpolate(A=@n31, B=Abs(None=@n31), Alpha=ScalarParameter() {defaultValue=1, parameterName=Abs the Fresnel, sliderMax=1})
  n33 = VectorParameter() {parameterName=Fesnel Options, channelNames {a=contrast, b=add, g=mul, r=pow}, defaultValue=<rgba 1 1 0 0>}
  n34 = Add(A=Multiply(A=Power(Base=@n32, Exp=@n33.pow), B=@n33.mul), B=@n33.add)
  LinearInterpolate(
   A=@n30
   B=Multiply(
    A=@n30
    B=MaterialFunctionCall(In=@n34, Contrast=@n33.contrast) {materialFunction=&CheapContrast}.Result
   )
   Alpha=ScalarParameter() {parameterName=Use Fresnel}
  )
  n35 = VectorParameter() {parameterName=SeondaryColor, defaultValue=<rgba 1 0.2448 0.2985 1>}
  n36 = Multiply(A=@n2, B=@n1)
  n37 = Multiply(
   A=@n7.RGB
   B=Desaturation(
    None=TextureSampleParameter2D(UVs=@n11) {parameterName=ColorTexture, texture=/Game/Effects/Fort_Effects/Textures/Smoke/T_Smoke_Trail_Soft_Chromatic}.RGB
    Fraction=ScalarParameter() {parameterName=Desaturate-ColorTexture}
   )
  )
  n38 = Multiply(A=Multiply(A=@n7.RGB, B=@n3), B=@n1)
  MaterialFunctionCall(Emissive=Multiply(A=LinearInterpolate(A=@n35.RGB, Alpha=@n36), B=Add(A=@n37, B=@n38))) {materialFunction=&MF_FX_EmissiveAdjustments}
  n39 = VertexTangentWS()
  n40 = LinearInterpolate(A=@n39, B=CrossProduct(A=@n39, B=VertexNormalWS()), Alpha=ScalarParameter() {parameterName=Vert Expantion Direction tangent vs binormal})
  n41 = MaterialFunctionCall() {materialFunction=&LinearGradient}
  n42 = LinearInterpolate(A=@n41.UGradient, B=@n41.UGradient, Alpha=ScalarParameter() {parameterName=Expation Axis})
  n43 = MaterialFunctionCall(Distance=ScalarParameter() {defaultValue=100, parameterName=NearCameraShrink}) {materialFunction=&Near_Camera_Fade}
  n44 = Multiply(A=@n18.SplineWidth, B=Multiply(A=ScalarParameter() {defaultValue=8, parameterName=SplineWidth}, B=Power(Base=@n43.Result) {constExponent=3}))
  n45 = Multiply(A=Multiply(A=Subtract(A=@n42) {constB=0.5}) {constB=2}, B=@n44)
  n46 = Multiply(A=ScalarParameter() {defaultValue=0.25, parameterName=SplineWidthTipMult}, B=@n44) {constB=5}
  Multiply(
   A=LinearInterpolate(
    A=Multiply(A=@n40, B=@n45)
    B=MaterialFunctionCall(WidthBase=@n46, WidthTip=@n44) {materialFunction=&SplineThicken}.WorldPositionOffset
    Alpha=ScalarParameter() {defaultValue=1, parameterName="Use Spline Thicken (otherwise use the alternate)"}
   )
   B=ScalarParameter() {defaultValue=1, parameterName=WPO Strength}
  )
 ]
}
