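# Translucent, two-sided material for wind updraft streaks on spline effects, with a Color parameter.
# SplineThicken controls strip width through world position offset, using base and tip width parameters.
# Two textures pan at WindSpeed, offset by vertex color red, and multiply with UV gradient masks that soften all edges.
# Opacity then fades by noise-varied depth intersection, near-camera distance, far pixel depth and view angle.
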Material /Game/Effects/Fort_Effects/Materials/Splines/M_Wind_Spline_Updraft
graph {
 properties {blendMode=BLEND_Translucent, pixelDepthOffsetMode=PDOM_Legacy, twoSided=true, usedShadingModels=Default Lit}
 nodes [
  n1 = MaterialFunctionCall(
   WidthBase=ScalarParameter() {defaultValue=350, group=SplineControl, parameterName=WidthBase}
   WidthTip=ScalarParameter() {defaultValue=200, group=SplineControl, parameterName=WidthTip}
  ) {materialFunction=&SplineThicken}
  n2 = Fresnel(Normal=@n1.Normal) {exponent=1}
  Clamp(None=Power(Base=OneMinus(None=@n2)))
  DistanceCullFade()
  n3 = MaterialFunctionCall() {materialFunction=&LinearGradient}
  n4 = Multiply(A=Multiply(A=OneMinus(None=@n3.UGradient), B=@n3.UGradient)) {constB=4}
  n5 = Multiply(A=Multiply(A=OneMinus(None=@n3.VGradient), B=@n3.VGradient)) {constB=4}
  n6 = VertexColor()
  n7 = TextureCoordinate() {uTiling=0.5, vTiling=0.75}
  n8 = Multiply(A=Time(), B=ScalarParameter() {defaultValue=1, parameterName=WindSpeed})
  n9 = TextureSample(UVs=Panner(Coordinate=Add(A=@n6.R, B=@n7), Time=@n8) {speedX=-0.05, speedY=1}) {texture=/Game/Global/Textures/Default/Masks_and_Gradients/T_BlasterBeamDistort}
  n10 = ScalarParameter() {defaultValue=0.05, parameterName=TextureContrast}
  n11 = Panner(Coordinate=Add(A=@n6.R, B=TextureCoordinate()), Time=@n8) {speedX=0.1, speedY=1}
  n12 = Multiply(A=Multiply(
    A=MaterialFunctionCall(In=@n9.G, Contrast=@n10) {materialFunction=&CheapContrast}.Result
    B=TextureSample(UVs=@n11) {texture=&T_NoisePack_01}.G
   ))
  n13 = ScalarParameter() {defaultValue=0.75, parameterName=AlphaBoost}
  n14 = DepthFade(
   Opacity=Multiply(A=Multiply(A=@n4, B=@n5), B=Multiply(A=@n12, B=@n13))
   FadeDistance=Multiply(A=OneMinus(None=@n9.G), B=ScalarParameter() {defaultValue=128, parameterName=FadeDistance})
  )
  n15 = MaterialFunctionCall(Distance=ScalarParameter() {defaultValue=1024, parameterName=CameraFadeDistance}) {materialFunction=&Near_Camera_Fade}
  n16 = PixelDepth()
  n17 = ScalarParameter() {defaultValue=5000, parameterName=FarDistanceOffset}
  n18 = Clamp(None=Divide(A=Subtract(A=@n16.R, B=@n17), B=ScalarParameter() {defaultValue=3000, parameterName=FarDistanceLength}))
  Multiply(
   A=Multiply(A=Multiply(A=@n14, B=@n15.Result), B=OneMinus(None=@n18))
   B=MaterialFunctionCall("FadeSensitivity (optional)"=ScalarParameter() {defaultValue=1, parameterName=FadeAngleSensitivity}) {materialFunction=&FadeBasedOnViewAngle}.Result
  )
  Reroute(None=@n1.WorldPositionOffset)
  VectorParameter() {parameterName=Color, defaultValue=<rgba 1.5 0.6341 0.3933 1>}
 ]
}
